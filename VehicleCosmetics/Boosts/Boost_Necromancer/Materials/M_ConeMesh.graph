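# Translucent unlit material for Niagara mesh particles, evidently the cone mesh of the Necromancer boost effect, with animated lightning-like energy.
# Panning samples of one caustics noise texture are combined into lightning and core masks, controlled by a NoCore switch.
# The result is colored through optional hue shifts, desaturation and particle color, then passed to MF_FX_EmissiveAdjustments.
# Named reroutes also define custom color brightness correction, gradient masks and a two-sine wave offset along vertex normals.
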Material /VehicleCosmetics/Boosts/Boost_Necromancer/Materials/M_ConeMesh
graph {
 properties {
  bAutomaticallySetUsageInEditor=false
  bUsedWithNiagaraMeshParticles=true
  blendMode=BLEND_Translucent
  shadingModel=MSM_Unlit
  usedShadingModels=Unlit
 }
 nodes [
  n1 = ParticleColor()
  n2 = ScalarParameter() {defaultValue=1, group=Color, parameterName=ColorScale}
  n3 = ScalarParameter() {defaultValue=10, parameterName=Brightness}
  n4 = NamedRerouteUsage()
  n5 = ScalarParameter() {defaultValue=1, parameterName=CoreBlend}
  n6 = NamedRerouteUsage()
  n7 = Multiply(A=TextureCoordinate() {vTiling=0.05}, B=NamedRerouteUsage())
  n8 = Constant() {r=0.25}
  n9 = TextureSample(UVs=Panner(Coordinate=Panner(Coordinate=@n7, Time=@n8) {speedX=-0.5, speedY=-0.5}, Time=NamedRerouteUsage()) {speedX=0.05, speedY=0.2}) {texture=/DealershipSharedLibrary/FX/Textures/Noise/T_VCVFX_CausticsMask_RGB}
  n10 = Constant() {r=0.35}
  n11 = LinearInterpolate(A=@n9.G, B=@n9.G, Alpha=@n10)
  n12 = TextureCoordinate() {vTiling=0.5}
  n13 = VectorParameter() {parameterName=Tile, defaultValue=<rgba 1 1 15 1>}
  n14 = NamedRerouteDeclaration(None=ComponentMask(None=@n13.RGB) {g=true, r=true}) {name=Tile - ComponentMask, nodeColor=<rgba 0.6406 0 1 1>}
  n15 = TextureSample(UVs=Panner(Coordinate=Multiply(A=@n12, B=@n14), Time=NamedRerouteUsage()) {speedX=0.05, speedY=0.2}) {texture=&T_VCVFX_CausticsMask_RGB}
  n16 = TextureCoordinate() {uTiling=-1, vTiling=0.5}
  n17 = TextureSample(UVs=Panner(Coordinate=Multiply(A=@n14, B=@n16), Time=NamedRerouteUsage()) {speedX=0.05, speedY=0.15}) {texture=&T_VCVFX_CausticsMask_RGB}
  n18 = Add(A=@n15.R, B=@n17.B)
  n19 = Add(A=@n15.B, B=@n17.R)
  n20 = Multiply(A=@n18, B=@n19)
  n21 = Multiply(A=TextureCoordinate() {vTiling=0.1}, B=NamedRerouteUsage())
  n22 = Constant() {r=0.8}
  n23 = TextureSample(UVs=Panner(Coordinate=Panner(Coordinate=@n21, Time=@n22) {speedX=-1, speedY=-0.5}, Time=NamedRerouteUsage()) {speedX=-0.05, speedY=0.15}) {texture=&T_VCVFX_CausticsMask_RGB}
  n24 = LinearInterpolate(A=LinearInterpolate(A=@n23.R, B=@n23.R, Alpha=Constant() {r=0.75}), B=@n23.G, Alpha=Constant() {r=0.5})
  n25 = Constant() {r=3}
  n26 = Add(A=@n15.G, B=@n17.G)
  n27 = Add(A=Multiply(A=LinearInterpolate(A=@n11, B=@n9.B, Alpha=@n10), B=@n20), B=Multiply(A=Power(Base=@n24, Exp=@n25), B=@n26))
  n28 = NamedRerouteUsage()
  n29 = Multiply(
   A=StaticSwitchParameter(True=@n27, False=Multiply(A=@n27, B=@n28)) {parameterName=NoCore}
   B=ScalarParameter() {defaultValue=50, parameterName=LightningBrightness}
  )
  n30 = Multiply(A=NamedRerouteUsage(), B=ScalarParameter() {defaultValue=3, parameterName=Core})
  n31 = Multiply(A=Add(A=@n9.R, B=@n23.R), B=Add(A=@n18, B=@n19))
  n32 = Add(A=Multiply(A=NamedRerouteUsage(), B=@n27), B=Saturate(None=ConstantBiasScale(None=@n31) {bias=-0.6}))
  n33 = LinearInterpolate(
   A=Add(A=Multiply(A=@n4, B=@n5), B=Multiply(A=@n6, B=@n29))
   B=Saturate(None=Subtract(A=@n30, B=@n32))
   Alpha=ScalarParameter() {defaultValue=0.5, parameterName=Opacity}
  )
  n34 = NamedRerouteUsage()
  n35 = Multiply(A=ScalarParameter() {defaultValue=0.1, parameterName=HueShit}, B=Saturate(None=LinearInterpolate(A=@n20, B=@n26, Alpha=@n34)))
  n36 = NamedRerouteUsage()
  n37 = StaticSwitchParameter(
   True=Saturate(None=MaterialFunctionCall(Hue Shift Percentage=@n35, Texture=@n36) {materialFunction=&HueShift}.Result)
   False=NamedRerouteUsage()
  ) {parameterName=HueShift}
  n38 = Multiply(A=NamedRerouteUsage(), B=@n37)
  n39 = Constant() {r=3}
  n40 = Desaturation(None=Saturate(None=Power(Base=@n38, Exp=@n39)), Fraction=ScalarParameter() {defaultValue=0.3, parameterName=Saturation})
  n41 = Add(A=@n20, B=@n26)
  n42 = Multiply(A=NamedRerouteUsage(), B=Saturate(None=ConstantBiasScale(None=@n41) {bias=-0.2, scale=2}))
  n43 = ScalarParameter() {parameterName=HueShiftSaturation}
  n44 = Multiply(A=Saturate(None=@n33), B=Add(A=Multiply(A=@n42, B=@n40), B=Desaturation(None=@n37, Fraction=@n43)))
  MaterialFunctionCall(Emissive=Multiply(A=Multiply(A=@n1.RGB, B=@n2), B=Multiply(A=@n3, B=@n44))) {materialFunction=&MF_FX_EmissiveAdjustments}
  NamedRerouteDeclaration(None=Multiply(A=ScalarParameter() {defaultValue=8, parameterName=PanSpeed}, B=Time())) {name=PanSpeed, nodeColor=<rgba 1 0.1641 0 1>}
  n45 = NamedRerouteUsage()
  n46 = ScalarParameter() {defaultValue=3, parameterName=GC_HueShift}
  n47 = MaterialFunctionCall(Hue Shift Percentage=Saturate(None=Multiply(A=@n45, B=@n46)), Texture=NamedRerouteUsage()) {materialFunction=&HueShift}
  n48 = ScalarParameter() {parameterName=GC_Saturation}
  NamedRerouteDeclaration(None=StaticSwitchParameter(True=Desaturation(None=@n47.Result, Fraction=@n48), False=NamedRerouteUsage()) {parameterName=CG}) {name=CG, nodeColor=<rgba 0.3594 0 1 1>}
  n49 = VectorParameter() {parameterName=CustomColor, defaultValue=<rgba 1 1 1 1>}
  NamedRerouteDeclaration(None=@n49.RGB) {name=CustomColor - RGB, nodeColor=<rgba 1 0 0.7734 1>}
  n50 = NamedRerouteUsage()
  n51 = Constant() {r=0.75}
  n52 = Constant() {r=0.35}
  n53 = Multiply(A=NamedRerouteUsage(), B=Constant() {r=2})
  n54 = Add(A=NamedRerouteUsage(), B=Constant() {r=0.05})
  n55 = Add(A=@n49.R, B=@n49.R)
  n56 = Saturate(None=Divide(A=Add(A=@n55, B=@n49.B), B=Constant() {r=3}))
  NamedRerouteDeclaration(None=If(
    A=@n56
    B=Constant() {r=0.785}
    "A > B"=Multiply(A=@n50, B=@n51)
    "A == B"=NamedRerouteUsage()
    "A < B"=If(A=@n56, B=@n52, "A > B"=@n54, "A == B"=@n54, "A < B"=@n53)
   )) {name=CustomColor, nodeColor=<rgba 0.3359 0 1 1>}
  n57 = ScalarParameter() {defaultValue=1, parameterName=Strength}
  n58 = Constant()
  n59 = TextureCoordinate() {coordinateIndex=1}
  n60 = Time()
  n61 = Add(
   A=Divide(A=ComponentMask(None=@n59) {r=true}, B=ScalarParameter() {defaultValue=0.5, group=WPO, parameterName=CountWaves})
   B=Multiply(A=ScalarParameter() {defaultValue=-0.5, group=WPO, parameterName=Speed}, B=OneMinus(None=@n60))
  )
  n62 = ComponentMask(None=TextureCoordinate()) {r=true}
  n63 = ScalarParameter() {defaultValue=0.5, group=WPO, parameterName=PowerMiddle}
  n64 = Multiply(
   A=Multiply(A=Sine(None=@n61) {period=0.5}, B=ScalarParameter() {defaultValue=1, parameterName=Strength})
   B=Multiply(A=Power(Base=@n62, Exp=@n63), B=ScalarParameter() {defaultValue=0.5, group=WPO, parameterName=MultPower})
  )
  n65 = ComponentMask(None=TextureCoordinate() {coordinateIndex=1}) {r=true}
  n66 = ScalarParameter() {defaultValue=0.5, group=WPO, parameterName=CountWaves}
  n67 = ScalarParameter() {defaultValue=-1, group=WPO, parameterName=SecondarySpeed}
  n68 = OneMinus(None=Time())
  n69 = Sine(None=Add(A=Divide(A=@n65, B=@n66), B=Multiply(A=@n67, B=@n68))) {period=0.2}
  n70 = AppendVector(A=AppendVector(A=@n58, B=@n58), B=Multiply(A=Add(A=@n64, B=@n69), B=Constant() {r=0.5}))
  NamedRerouteDeclaration(None=Multiply(A=Multiply(A=@n57, B=@n70), B=VertexNormalWS())) {name=WPO, nodeColor=<rgba 0 1 0.9531 1>}
  n71 = Multiply(A=TextureCoordinate() {uTiling=0.5}, B=AppendVector(A=Constant() {r=2}, B=Constant() {r=1}))
  n72 = Add(A=ComponentMask(None=Frac(None=@n71)) {g=true}, B=Time())
  n73 = Saturate(None=ConstantBiasScale(None=Sine(None=@n72) {period=2}) {bias=-0.1, scale=1})
  NamedRerouteDeclaration(None=Multiply(A=@n73, B=@n73)) {name=GradientMask, nodeColor=<rgba 1 0.1641 0 1>}
  NamedRerouteDeclaration(None=@n13.B) {name=Tile - BChannel, nodeColor=<rgba 0.6406 0 1 1>}
  n74 = ComponentMask(None=TextureCoordinate()) {g=true}
  n75 = Saturate(None=ConstantBiasScale(None=OneMinus(None=@n74)) {bias=0, scale=8})
  n76 = Saturate(None=ConstantBiasScale(None=@n74) {bias=-0.025, scale=2})
  NamedRerouteDeclaration(None=Saturate(None=Multiply(A=@n75, B=@n76))) {name=Emissive _ GradientMask, nodeColor=<rgba 0 1 0.4609 1>}
  NamedRerouteDeclaration(None=@n33) {name=Emissive - Lerp, nodeColor=<rgba 1 0.4453 0 1>}
  n77 = TextureCoordinate()
  n78 = Multiply(A=Power(Base=ComponentMask(None=@n77) {g=true}, Exp=Constant() {r=4}), B=Constant() {r=4})
  n79 = LinearInterpolate(
   A=ScalarParameter() {defaultValue=2, parameterName=FresnelEnd}
   B=ScalarParameter() {defaultValue=3, parameterName=FresnelBase}
   Alpha=NamedRerouteDeclaration(None=Saturate(None=@n78)) {name=SecondaryMask, nodeColor=<rgba 0.9922 1 0 1>}
  )
  NamedRerouteDeclaration(None=Multiply(A=NamedRerouteUsage(), B=Saturate(None=@n79))) {name=LightDirection, nodeColor=<rgba 0 0.4375 1 1>}
  NamedRerouteUsage()
  Saturate(None=Multiply(A=ScalarParameter() {defaultValue=3, parameterName=Hardness}, B=NamedRerouteUsage()))
 ]
}
