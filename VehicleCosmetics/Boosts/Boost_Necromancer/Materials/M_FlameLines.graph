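# Masked unlit material for Niagara mesh particles in the Necromancer boost, drawing flickering animated bolt-circle lines.
# Two time-stepped flipbook samples, one U-mirrored, rotate independently.
# Zero-tiled panning cloud noise, thresholded into uniform switches, flickers between their channels.
# The inverted result, shaped by dynamic Power, tints emissive with particle color or InitColor via isPainted, while particle alpha and Hardness shape the mask.
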Material /VehicleCosmetics/Boosts/Boost_Necromancer/Materials/M_FlameLines
graph {
 properties {
  bAutomaticallySetUsageInEditor=false
  bUsedWithNiagaraMeshParticles=true
  blendMode=BLEND_Masked
  shadingModel=MSM_Unlit
  usedShadingModels=Unlit
 }
 nodes [
  n1 = ScalarParameter() {defaultValue=2, parameterName=Columns}
  n2 = ScalarParameter() {defaultValue=2, parameterName=Row}
  n3 = AppendVector(A=@n1, B=@n2)
  n4 = Divide(A=TextureCoordinate() {uTiling=-1, vTiling=0.9}, B=@n3)
  n5 = AppendVector(A=Multiply(A=@n1, B=@n2), B=@n2)
  n6 = Frac(None=Multiply(A=Time(), B=ScalarParameter() {defaultValue=1, parameterName=Speed}))
  n7 = AppendVector(A=@n6, B=@n6)
  n8 = Divide(A=Floor(None=Multiply(A=@n5, B=@n7)), B=@n3)
  n9 = TextureSample(UVs=MaterialFunctionCall(UVs=Add(A=@n4, B=@n8), "Rotation Angle (0-1)"=ScalarParameter() {parameterName=TextureARotation}) {materialFunction=&CustomRotator}.Rotated Values) {texture=/DealershipSharedLibrary/FX/Textures/Flipbook/T_VCVFX_BoltCircle}
  n10 = TextureCoordinate() {uTiling=0, vTiling=0}
  n11 = MaterialFunctionCall(
   X=TextureSample(UVs=Panner(Coordinate=@n10) {speedX=1, speedY=0.65}) {texture=/VehicleCosmetics/FX/Textures/Noise/T_VCVFX_Clouds_A}.R
   Y=TextureSample(UVs=Panner(Coordinate=@n10) {speedX=-1, speedY=-0.3}) {texture=&T_VCVFX_Clouds_A}.R
   Z=TextureSample(UVs=Panner(Coordinate=@n10) {speedX=-0.5, speedY=0.3}) {texture=&T_VCVFX_Clouds_A}.R
  ) {materialFunction=&MakeFloat3}
  n12 = Constant() {r=0.2}
  n13 = Saturate(None=Multiply(A=Subtract(A=@n11.Result, B=@n12), B=Constant() {r=35}))
  n14 = ComponentMask(None=@n13) {r=true}
  n15 = ComponentMask(None=@n13) {g=true}
  n16 = Divide(A=TextureCoordinate() {vTiling=0.9}, B=@n3)
  n17 = TextureSample(UVs=MaterialFunctionCall(UVs=Add(A=@n16, B=@n8), "Rotation Angle (0-1)"=ScalarParameter() {parameterName=TextureBRotation}) {materialFunction=&CustomRotator}.Rotated Values) {texture=&T_VCVFX_BoltCircle}
  n18 = LinearInterpolate(
   A=LinearInterpolate(A=LinearInterpolate(A=@n9.R, B=@n9.R, Alpha=@n14), B=@n9.B, Alpha=@n15)
   B=LinearInterpolate(A=LinearInterpolate(A=@n17.G, B=@n17.G, Alpha=@n14), B=@n17.R, Alpha=@n15)
   Alpha=ComponentMask(None=@n13) {b=true}
  )
  n19 = ConstantBiasScale(None=@n18) {bias=0}
  n20 = Saturate(None=OneMinus(None=Saturate(None=@n19)))
  n21 = DynamicParameter() {defaultValue=<rgba 0 1 1 1>, paramNames [isPainted?, Power, Param3, Hardness]}
  n22 = ParticleColor()
  n23 = Multiply(A=@n22.RGB, B=ScalarParameter() {defaultValue=4, parameterName=ScaleColor})
  n24 = Multiply(
   A=VectorParameter() {group=Color, parameterName=InitColor, defaultValue=<rgba 0.003 0 0.011 1>}.RGB
   B=ScalarParameter() {defaultValue=4, parameterName=ScaleColor}
  )
  MaterialFunctionCall(Emissive=Multiply(A=Power(Base=@n20, Exp=@n21.Power), B=If(A=@n21.isPainted?, "A > B"=@n23, "A == B"=@n24, "A < B"=@n24))) {materialFunction=&MF_FX_EmissiveAdjustments}
  n25 = Saturate(None=ConstantBiasScale(None=@n18) {bias=0, scale=8})
  Multiply(A=@n22.A, B=Saturate(None=Multiply(A=@n21.Hardness, B=@n25)))
 ]
}
