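# Material function for the Necromancer boost main trail, outputting Emissive from particle color and an Opacity built from a panning RGB trail texture.
# Static bool inputs pick and sum texture channels, shaped by particle alpha and power fades at both U ends.
# The "Is Masked?" switch flips pan direction, changes tiling and outputs saturated shape directly.
# Otherwise, panning tip mask and blob noise erode it, with U falloff, depth and camera-depth fades.
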MaterialFunction /VehicleCosmetics/Boosts/Boost_Necromancer/Materials/MF_MainTrail
graph {
 properties {previewBlendMode=BLEND_Translucent}
 nodes [
  n1 = Constant() {r=0.1}
  n2 = AppendVector(
   A=FunctionInput() {bUsePreviewValueAsDefault=true, inputName=UOffset, inputType=FunctionInput_Scalar, sortPriority=7}
   B=FunctionInput() {bUsePreviewValueAsDefault=true, inputName=YOffset, inputType=FunctionInput_Scalar, sortPriority=8}
  )
  n3 = TextureCoordinate()
  n4 = Constant2Vector() {g=0.875, r=0.65}
  n5 = Constant2Vector() {g=1, r=0.65}
  n6 = FunctionInput() {bUsePreviewValueAsDefault=true, inputName=UTiling, inputType=FunctionInput_Scalar, sortPriority=5, previewValue=<1 0 0 1>}
  n7 = FunctionInput() {bUsePreviewValueAsDefault=true, inputName=YTiling, inputType=FunctionInput_Scalar, sortPriority=6, previewValue=<1 0 0 1>}
  n8 = Multiply(A=StaticSwitchParameter(True=Multiply(A=@n3, B=@n4), False=Multiply(A=@n3, B=@n5)) {parameterName=Is Masked?}, B=AppendVector(A=@n6, B=@n7))
  n9 = FunctionInput() {bUsePreviewValueAsDefault=true, inputName=Speed, inputType=FunctionInput_Scalar, sortPriority=9}
  n10 = Time()
  n11 = Constant() {r=-1.5}
  n12 = StaticSwitchParameter(True=Abs(None=@n11), False=@n11) {parameterName=Is Masked?}
  n13 = Constant()
  n14 = TextureSample(UVs=Panner(Coordinate=Add(A=@n2, B=@n8), Time=Multiply(A=@n9, B=@n10), Speed=AppendVector(A=@n12, B=@n13)) {speedX=-1.5}) {texture=/DealershipSharedLibrary/FX/Textures/TexturePackage/T_VCVFX_NecroTrailsRGB}
  n15 = Constant()
  n16 = StaticBool()
  n17 = FunctionInput(Preview=@n16) {bUsePreviewValueAsDefault=true, inputName=R, inputType=FunctionInput_StaticBool}
  n18 = FunctionInput(Preview=StaticBool() {value=true}) {bUsePreviewValueAsDefault=true, inputName=G, inputType=FunctionInput_StaticBool, sortPriority=1}
  n19 = FunctionInput(Preview=@n16) {bUsePreviewValueAsDefault=true, inputName=B, inputType=FunctionInput_StaticBool, sortPriority=2}
  n20 = Add(
   A=Add(A=StaticSwitch(True=@n14.R, False=@n15, Value=@n17), B=StaticSwitch(True=@n14.G, False=@n15, Value=@n18))
   B=Multiply(A=StaticSwitch(True=@n14.B, False=@n15, Value=@n19), B=Constant() {r=3})
  )
  n21 = FunctionInput() {bUsePreviewValueAsDefault=true, inputName=ParticleAlpha, inputType=FunctionInput_Scalar, sortPriority=4, previewValue=<1 0 0 1>}
  n22 = ComponentMask(None=TextureCoordinate()) {r=true}
  n23 = Constant() {r=1.3}
  n24 = FunctionInput() {bUsePreviewValueAsDefault=true, inputName=PowerFade, inputType=FunctionInput_Scalar, sortPriority=10, previewValue=<1.6 0 0 1>}
  n25 = Multiply(A=Multiply(A=Power(Base=@n20, Exp=@n1), B=ComponentMask(None=@n21) {a=true}), B=Power(Base=Multiply(A=@n22, B=@n23), Exp=@n24))
  n26 = TextureCoordinate()
  n27 = FunctionInput() {bUsePreviewValueAsDefault=true, inputName=PowerFadeOneMinus, inputType=FunctionInput_Scalar, sortPriority=11, previewValue=<1.5 0 0 1>}
  n28 = Power(Base=OneMinus(None=ComponentMask(None=@n26) {r=true}), Exp=@n27)
  n29 = Multiply(A=Multiply(A=Multiply(A=@n25, B=@n28), B=Constant() {r=3}), B=@n20)
  n30 = Constant() {r=1.3}
  n31 = Constant() {r=20}
  n32 = Multiply(A=Multiply(A=@n29, B=@n30), B=@n31)
  n33 = Time()
  n34 = FunctionInput() {bUsePreviewValueAsDefault=true, inputName=EndDissolve, inputType=FunctionInput_Scalar, sortPriority=12}
  n35 = TextureSample(UVs=Panner(Coordinate=TextureCoordinate() {vTiling=0.0125}, Time=Multiply(A=@n33, B=@n34)) {speedY=1}) {texture=/DealershipSharedLibrary/FX/Textures/Gradient_Masks/T_VCVFX_RibbonTipMask_A}
  n36 = TextureCoordinate() {uTiling=1.4, vTiling=0.7}
  n37 = Panner(Coordinate=TextureCoordinate()) {speedX=-1, speedY=0.5}
  n38 = Power(Base=ComponentMask(None=TextureCoordinate()) {r=true}, Exp=Constant() {r=12})
  n39 = Constant() {r=200}
  n40 = Multiply(
   A=TextureSample(UVs=Panner(Coordinate=@n36) {speedX=-2, speedY=0.25}) {texture=/DealershipSharedLibrary/FX/Textures/Noise/T_VCVFX_BlobNoiseGlow_A}.RGB
   B=Multiply(A=TextureSample(UVs=@n37) {texture=&T_VCVFX_BlobNoiseGlow_A}.RGB, B=Multiply(A=@n38, B=@n39))
  )
  n41 = Multiply(A=Subtract(A=@n29, B=Add(A=@n35.RGB, B=@n40)), B=@n30)
  n42 = ComponentMask(None=TextureCoordinate()) {r=true}
  n43 = Multiply(A=Power(Base=OneMinus(None=@n42), Exp=Constant() {r=3}), B=@n31)
  n44 = Multiply(
   A=Multiply(A=Multiply(A=@n41, B=@n43), B=DepthFade() {fadeDistanceDefault=15})
   B=MaterialFunctionCall(Fade Length=Constant() {r=100}, Fade Offset=Constant() {r=50}) {materialFunction=&CameraDepthFade}.Result
  )
  FunctionOutput(None=StaticSwitchParameter(True=Saturate(None=@n32), False=Saturate(None=@n44)) {parameterName=Is Masked?}) {outputName=Opacity, sortPriority=1}
  n45 = FunctionInput() {bUsePreviewValueAsDefault=true, inputName=ParticleColor, sortPriority=3, previewValue=<1 1 1 1>}
  FunctionOutput(None=ComponentMask(None=@n45) {b=true, g=true, r=true}) {outputName=Emissive}
 ]
}
