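# Translucent unlit material for the main trail of the Necromancer vehicle boost, used on Niagara ribbons and mesh particles.
# Four MF_MainTrail layers (dark, slow, fast, light) run at different speeds, with the light layer also getting a UV-band mask.
# Layer emissives are blended by chained lerps into MF_FX_EmissiveAdjustments, alongside an additive four-input sum.
# Particle color and dynamic parameters expose paint state, light-trail multiplier and offset, with conditional color overrides on three layers.
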Material /VehicleCosmetics/Boosts/Boost_Necromancer/Materials/M_MainTrail
graph {
 properties {
  bAutomaticallySetUsageInEditor=false
  bUsedWithNiagaraMeshParticles=true
  bUsedWithNiagaraRibbons=true
  blendMode=BLEND_Translucent
  shadingModel=MSM_Unlit
  usedShadingModels=Unlit
 }
 nodes [
  n1 = NamedRerouteUsage()
  n2 = NamedRerouteUsage()
  Add(A=Add(A=Add(A=@n1, B=@n2), B=NamedRerouteUsage()), B=NamedRerouteUsage())
  n3 = LinearInterpolate(A=LinearInterpolate(B=NamedRerouteUsage(), Alpha=NamedRerouteUsage()), B=NamedRerouteUsage(), Alpha=NamedRerouteUsage())
  n4 = NamedRerouteUsage()
  n5 = NamedRerouteUsage()
  MaterialFunctionCall(Emissive=LinearInterpolate(A=LinearInterpolate(A=@n3, B=@n4, Alpha=@n5), B=NamedRerouteUsage(), Alpha=NamedRerouteUsage())) {materialFunction=&MF_FX_EmissiveAdjustments}
  n6 = ComponentMask(None=TextureCoordinate()) {g=true}
  n7 = Multiply(A=OneMinus(None=@n6), B=@n6)
  NamedRerouteDeclaration(None=Multiply(A=NamedRerouteUsage(), B=Step(X=@n7) {constY=0.23})) {name=LightTrail - Mask, nodeColor=<rgba 0 0.9297 1 1>}
  n8 = StaticBool()
  n9 = NamedRerouteUsage()
  n10 = NamedRerouteUsage()
  n11 = VectorParameter() {parameterName=LightTrail - Color, defaultValue=<rgba 0.032 0.045 0.13 1>}
  n12 = MaterialFunctionCall!(
   R=@n8
   G=@n8
   B=StaticBool() {value=true}
   ParticleColor=Multiply(A=If(A=@n9, "A > B"=@n10, "A == B"=@n11.RGB, "A < B"=@n11.RGB), B=NamedRerouteUsage())
   ParticleAlpha=Add(A=NamedRerouteUsage()) {constB=0.5}
   UTiling
   YTiling=Constant() {r=4}
   UOffset=NamedRerouteUsage()
   YOffset=ScalarParameter() {defaultValue=-0.4, parameterName=LightTrail - YOffset}
   Speed=Constant() {r=0.9}
   PowerFade
   PowerFadeOneMinus
   EndDissolve=Constant()
  ) {materialFunction=&/MF_MainTrail}
  NamedRerouteDeclaration(None=@n12.Opacity) {name=LightTrail - Opacity, nodeColor=<rgba 0 0.9297 1 1>}
  n13 = StaticBool()
  n14 = MaterialFunctionCall(
   R=StaticBool() {value=true}
   G=@n13
   B=@n13
   ParticleColor=Constant3Vector() {constant=<rgba 0.001 0 0.011 1>}
   ParticleAlpha=NamedRerouteUsage()
   UOffset=NamedRerouteUsage()
   YOffset=ScalarParameter() {parameterName=Default YOffset}
   Speed=Constant() {r=0.7}
   EndDissolve=Constant() {r=0.4}
  ) {materialFunction=&MF_MainTrail}
  NamedRerouteDeclaration(None=@n14.Emissive) {name=DarkTrail - Emissive, nodeColor=<rgba 0 0.9297 1 1>}
  NamedRerouteDeclaration(None=@n14.Opacity) {name=DarkTrail - Opacity, nodeColor=<rgba 0 0.9297 1 1>}
  n15 = ParticleColor()
  NamedRerouteDeclaration(None=@n15.RGB) {name=ParticleRGB, nodeColor=<rgba 0 0.6953 1 1>}
  NamedRerouteDeclaration(None=DynamicParameter() {parameterIndex=1, defaultValue=<rgba 0 0 0 0>, paramNames [isPainted?, Param2, Param3, Param4]}.isPainted?) {name=Dynamic - isPainted?, nodeColor=<rgba 1 0 0.375 1>}
  n16 = DynamicParameter() {paramNames [LightTrailMultiplier, SlowTrailMultiplier, Offset, Param4]}
  NamedRerouteDeclaration(None=@n16.Offset) {name=Offset, nodeColor=<rgba 0 1 0.1094 1>}
  NamedRerouteDeclaration(None=@n12.Emissive) {name=LightTrail - Emissive, nodeColor=<rgba 0 0.9297 1 1>}
  n17 = NamedRerouteUsage()
  n18 = NamedRerouteUsage()
  n19 = VectorParameter() {parameterName=SlowTrail - Color, defaultValue=<rgba 0.03 0.02 0.08 1>}
  n20 = MaterialFunctionCall(
   ParticleColor=Multiply(A=If(A=@n17, "A > B"=@n18, "A == B"=@n19.RGB, "A < B"=@n19.RGB), B=ScalarParameter() {defaultValue=0.9, parameterName=SlowTrailColorScale})
   ParticleAlpha=NamedRerouteUsage()
   UOffset=NamedRerouteUsage()
   YOffset=ScalarParameter() {parameterName=Default YOffset}
   Speed=Constant() {r=0.4}
   EndDissolve=Constant() {r=0.4}
  ) {materialFunction=&MF_MainTrail}
  NamedRerouteDeclaration(None=@n20.Emissive) {name=SlowTrail - Emissive, nodeColor=<rgba 0 0.9297 1 1>}
  NamedRerouteDeclaration(None=@n16.LightTrailMultiplier) {name=Dynamic - LightTrailMultiplier, nodeColor=<rgba 1 0 0.375 1>}
  NamedRerouteDeclaration(None=@n15.A) {name=ParticleAlpha, nodeColor=<rgba 0 0.6953 1 1>}
  NamedRerouteDeclaration(None=@n20.Opacity) {name=SlowTrail - Opacity, nodeColor=<rgba 0 0.9297 1 1>}
  n21 = NamedRerouteUsage()
  n22 = NamedRerouteUsage()
  n23 = VectorParameter() {parameterName=FastTrail - Color, defaultValue=<rgba 0.04 0 0.07 1>}
  n24 = MaterialFunctionCall(
   ParticleColor=Multiply(A=If(A=@n21, "A > B"=@n22, "A == B"=@n23.RGB, "A < B"=@n23.RGB), B=ScalarParameter() {defaultValue=1, parameterName=FastTrailColorScale})
   ParticleAlpha=NamedRerouteUsage()
   UOffset=NamedRerouteUsage()
   YOffset=ScalarParameter() {parameterName=Default YOffset}
   Speed=Constant() {r=1}
   EndDissolve=Constant() {r=0.4}
  ) {materialFunction=&MF_MainTrail}
  NamedRerouteDeclaration(None=@n24.Emissive) {name=FastTrail - Emissive, nodeColor=<rgba 0 0.9297 1 1>}
  NamedRerouteDeclaration(None=@n24.Opacity) {name=FastTrail - Opacity, nodeColor=<rgba 0 0.9297 1 1>}
 ]
}
