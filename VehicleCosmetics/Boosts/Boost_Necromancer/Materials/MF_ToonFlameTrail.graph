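# Material function for a toon-style flame trail in the Necromancer boost effect, outputting Opacity and Emissive.
# A panning flow texture distorts both a dot-product radial shape and a scrolling flame mask, with inputs for tiling, offset, distortion and speed.
# Opacity erodes the shape with the mask's green channel, sharpens the edge and fades it along U.
# Emissive quantizes the blue channel plus a U gradient into two tones with Ceil, tinted by ParticleColor.
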MaterialFunction /VehicleCosmetics/Boosts/Boost_Necromancer/Materials/MF_ToonFlameTrail
graph {
 properties {previewBlendMode=BLEND_Translucent}
 nodes [
  n1 = TextureCoordinate() {uTiling=0.8}
  n2 = FunctionInput() {bUsePreviewValueAsDefault=true, inputName=YTiling, inputType=FunctionInput_Scalar, previewValue=<1 0 0 1>}
  n3 = AppendVector(A=Constant() {r=1}, B=@n2)
  n4 = Panner(Coordinate=TextureCoordinate() {uTiling=0.5, vTiling=0.5}) {speedX=0.5}
  n5 = FunctionInput() {bUsePreviewValueAsDefault=true, inputName=Distort, inputType=FunctionInput_Scalar, sortPriority=2, previewValue=<1 0 0 1>}
  n6 = Multiply(
   A=ComponentMask(None=TextureSample(UVs=@n4) {texture=/DealershipSharedLibrary/FX/Textures/Noise/T_VCVFX_Flow02_N}.RGB) {g=true, r=true}
   B=@n5
  )
  n7 = Constant() {r=0.3}
  n8 = FunctionInput() {bUsePreviewValueAsDefault=true, inputName=TextureOffset, inputType=FunctionInput_Vector2, previewValue=<-0.45 -0.5 0 1>}
  n9 = Add(A=Add(A=Multiply(A=@n1, B=@n3), B=Multiply(A=@n6, B=@n7)), B=@n8)
  n10 = OneMinus(None=DotProduct(A=@n9, B=@n9))
  n11 = Constant() {r=0.5}
  n12 = FunctionInput() {bUsePreviewValueAsDefault=true, inputName=Tile, inputType=FunctionInput_Scalar, sortPriority=3, previewValue=<1 0 0 1>}
  n13 = Multiply(A=TextureCoordinate() {uTiling=0.3, vTiling=0.6}, B=@n12)
  n14 = FunctionInput() {bUsePreviewValueAsDefault=true, inputName=Speed, inputType=FunctionInput_Scalar, sortPriority=1, previewValue=<1 0 0 1>}
  n15 = Multiply(A=Time(), B=@n14)
  n16 = Constant() {r=0.1}
  n17 = TextureSample(UVs=Add(A=Panner(Coordinate=@n13, Time=@n15) {speedX=0.5}, B=Multiply(A=@n6, B=@n16))) {texture=/DealershipSharedLibrary/FX/Textures/TexturePackage/T_VCVFX_FlameMask_RGB}
  n18 = Constant() {r=2}
  n19 = Saturate(None=ComponentMask(None=TextureCoordinate()) {r=true})
  n20 = Subtract(A=Saturate(None=Subtract(A=@n10, B=@n11)), B=LinearInterpolate(A=Multiply(A=@n17.G, B=@n18), B=Constant(), Alpha=@n19))
  n21 = Multiply(A=Subtract(A=Saturate(None=@n20), B=Constant() {r=0.333}), B=Constant() {r=45})
  n22 = Multiply(A=OneMinus(None=@n19), B=Constant() {r=6})
  FunctionOutput(None=Multiply(A=Saturate(None=@n21), B=Saturate(None=@n22))) {outputName=Opacity}
  n23 = NamedRerouteUsage()
  n24 = Constant() {r=0.6}
  n25 = Add(A=Constant() {r=0.4}, B=Ceil(None=Subtract(A=@n23, B=@n24)))
  FunctionOutput(None=Multiply(A=Saturate(None=@n25), B=FunctionInput() {inputName=ParticleColor})) {outputName=Emissive}
  n26 = Multiply(A=Subtract(A=@n19, B=Constant() {r=0.75}), B=Constant() {r=2})
  NamedRerouteDeclaration(None=Add(A=@n17.B, B=Saturate(None=@n26))) {name=Opacity - Texture+Mask, nodeColor=<rgba 0.3594 1 0 1>}
 ]
}
