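# Unlit translucent Niagara sprite material for the toon flame trail of the Necromancer boost.
# Two MF_ToonFlameTrail layers, a dark outline and an offset, scaled particle-color flame, blend using their own emissive as alpha, then pass through MF_FX_EmissiveAdjustments.
# Opacity sums both layers, subtracts a mask, then multiplies by particle alpha and an optional view-angle fade.
# Particle dynamic parameters supply Speed, Distort and Tile through named reroutes.
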Material /VehicleCosmetics/Boosts/Boost_Necromancer/Materials/M_ToonFlameTrail
graph {
 properties {
  bAutomaticallySetUsageInEditor=false
  bUsedWithNiagaraSprites=true
  blendMode=BLEND_Translucent
  shadingModel=MSM_Unlit
  usedShadingModels=Unlit
 }
 nodes [
  Constant3Vector() {constant=<rgba 0.32 0.4 1 0>}
  n1 = MaterialFunctionCall!(
   ParticleColor=VectorParameter() {parameterName=DarkOutline, defaultValue=<rgba 0.001 0 0.011 1>}.RGB
   TextureOffset
   YTiling
   Speed=NamedRerouteUsage()
   Distort=NamedRerouteUsage()
   Tile=NamedRerouteUsage()
  ) {materialFunction=&/MF_ToonFlameTrail}
  n2 = MaterialFunctionCall(
   ParticleColor=Multiply(A=ParticleColor().RGB, B=ScalarParameter() {defaultValue=3, parameterName=ScaleColor})
   TextureOffset=Constant2Vector() {g=-0.8, r=-0.495}
   YTiling=Constant() {r=1.6}
   Speed=NamedRerouteUsage()
   Distort=NamedRerouteUsage()
   Tile=NamedRerouteUsage()
  ) {materialFunction=&MF_ToonFlameTrail}
  MaterialFunctionCall(Emissive=LinearInterpolate(A=LinearInterpolate(B=@n1.Emissive, Alpha=@n1.Emissive), B=@n2.Emissive, Alpha=@n2.Emissive)) {materialFunction=&MF_FX_EmissiveAdjustments}
  n3 = ScalarParameter() {parameterName=SubtractMask}
  n4 = Add(A=@n1.Opacity, B=@n2.Opacity)
  n5 = NamedRerouteUsage()
  Multiply(
   A=Multiply(A=Subtract(A=@n4, B=@n3) {constB=0.01}, B=ParticleColor().A)
   B=If(A=ScalarParameter() {parameterName=UseAngleFade?}, "A > B"=@n5, "A == B"=@n5, "A < B"=Constant() {r=1}) {constB=1}
  )
  n6 = ScalarParameter() {defaultValue=0.5, parameterName=Angle Fade - Transparent Facing Ratio}
  n7 = Divide(A=Constant() {desc=New Max, r=1}, B=Subtract(A=ScalarParameter() {defaultValue=0.5, parameterName=Angle Fade - Opaque Facing Ratio}, B=@n6))
  n8 = VertexNormalWS()
  n9 = StaticSwitch(
   True=CameraVectorWS()
   False=Transform(None=Constant3Vector() {desc=View Direction, constant=<rgba 0 0 -1 0>}) {transformSourceType=TRANSFORMSOURCE_View}
   Value=StaticBool() {value=true}
  )
  n10 = Add(A=Abs(None=DotProduct(A=@n8, B=@n9)), B=Subtract(A=Constant() {desc=New Min}, B=@n6))
  NamedRerouteDeclaration(None=Clamp(None=Multiply(A=@n7, B=@n10))) {name=AngleFade, nodeColor=<rgba 0.4816 0 1 1>}
  n11 = DynamicParameter() {paramNames [Speed,Distort,Tile,Param4]}
  NamedRerouteDeclaration(None=@n11.Speed) {name=Dynamic - Speed, nodeColor=<rgba 1 0 0.9141 1>}
  NamedRerouteDeclaration(None=@n11.Distort) {name=Dynamic - Distort, nodeColor=<rgba 1 0 0.9141 1>}
  NamedRerouteDeclaration(None=@n11.Tile) {name=Dynamic - Tile, nodeColor=<rgba 1 0 0.9141 1>}
 ]
}
